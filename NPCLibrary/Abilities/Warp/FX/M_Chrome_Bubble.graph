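# Masked, two-sided lit material for Niagara mesh particles, rendering a chrome bubble for the NPC Warp ability FX.
# Instance-space world-aligned cell noise supplies surface normals and combines with Fresnel for particle-tinted emissive that differs between front and back faces.
# The LackLuster base function supplies base color and roughness.
# Dynamic parameters drive erosion, noise tiling and normal-direction vertex offset, shaped by a vertical mask built from particle-center distance and normal direction.
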Material /NPCLibrary/Abilities/Warp/FX/M_Chrome_Bubble
graph {
 properties {
  bUsedWithNiagaraMeshParticles=true
  blendMode=BLEND_Masked
  opacityMaskClipValue=0.01
  pixelDepthOffsetMode=PDOM_Legacy
  twoSided=true
  usedShadingModels=Default Lit
 }
 nodes [
  ActorPositionWS()
  Constant() {r=1}
  n1 = NamedRerouteUsage()
  n2 = Fresnel() {baseReflectFraction=0, exponent=1}
  n3 = MaterialFunctionCall(In=@n2, Contrast=Multiply(A=NamedRerouteUsage(), B=Constant() {r=0.25})) {materialFunction=&CheapContrast}
  n4 = DynamicParameter() {defaultValue=<rgba 1 1 0 1>, paramNames [Erode 1-0, NoiseTIle, WPO, Param4]}
  n5 = Multiply(A=@n4.NoiseTIle, B=ScalarParameter() {defaultValue=60, parameterName=ErodeTile})
  n6 = TransformPosition(None=WorldPosition() {worldPositionShaderOffset=WPT_ExcludeAllShaderOffsets}.XYZ) {transformSourceType=TRANSFORMPOSSOURCE_World, transformType=TRANSFORMPOSSOURCE_Instance}
  n7 = MaterialFunctionCall(
   TextureObject=TextureObject() {samplerType=SAMPLERTYPE_LinearGrayscale, texture=/Game/Effects/Fort_Effects/Textures/Tile/T_CellNoise_Tile_M}
   TextureSize=@n5
   WorldPosition=@n6
  ) {materialFunction=/Game/Weapons/WeaponSkins/Wraps/MaterialFunctions/WorldAlignedTexture}
  n8 = OneMinus(None=@n2)
  n9 = Add(
   A=NamedRerouteDeclaration(None=ComponentMask(None=@n7.Z Texture) {g=true}) {name=Texture, nodeColor=<rgba 0 1 0.9297 1>}
   B=Multiply(A=Power(Base=@n8) {constExponent=8}) {constB=0.5}
  )
  n10 = LinearInterpolate(A=@n9, Alpha=@n3.Result) {constB=0}
  n11 = LinearInterpolate(
   A=Multiply(A=ConstantBiasScale(None=@n1) {bias=0.5, scale=2}, B=OneMinus(None=@n3.Result))
   B=Multiply(A=Power(Base=@n10), B=NamedRerouteUsage())
   Alpha=ConstantBiasScale(None=TwoSidedSign())
  )
  MaterialFunctionCall(Emissive=Multiply(A=Saturate(None=@n11), B=ParticleColor().RGB)) {materialFunction=&MF_FX_EmissiveAdjustments}
  MaterialFunctionCall() {materialFunction=&DebugTimeSine}
  MaterialFunctionCall() {materialFunction=&ObjectScale}
  n12 = BreakMaterialAttributes!(Attr=MaterialFunctionCall!(
    bBlendMaterialAttributes
    MA_In
    UVsOverride
    Override Normals
    BlendMaterialAttributesAlpha
    BaseColorOverride
    WorldPositionOverride
    VertexNormalOverride
    SourceNormalsStrengthOverride
   ) {materialFunction=/Game/TechArt/Seasonal/LackLuster/MaterialFunctions/MF_LackLuster_Base}.MA Out)
  Multiply(A=@n12.BaseColor, B=LinearInterpolate(Alpha=NamedRerouteUsage()) {constA=0.6})
  Multiply(A=@n12.Roughness, B=LinearInterpolate(Alpha=NamedRerouteUsage()) {constA=1, constB=0.8})
  n13 = WorldPosition()
  n14 = ParticlePositionWS()
  n15 = ScalarParameter() {defaultValue=15, parameterName=MeshRadius}
  n16 = LinearInterpolate(Alpha=Divide(A=Distance(A=@n13.XYZ, B=@n14), B=Multiply(A=@n15, B=@n4.Param4))) {constA=-1}
  n17 = DotProduct(
   A=Transform(None=VertexNormalWS()) {transformSourceType=TRANSFORMSOURCE_World, transformType=TRANSFORM_Instance}
   B=Constant3Vector() {constant=<rgba 0 0 1 0>}
  )
  n18 = Power(Base=Saturate(None=LinearInterpolate(A=@n16, B=@n17) {constAlpha=0.3}))
  n19 = LinearInterpolate(Alpha=@n18) {constA=-1}
  Multiply(A=@n4.WPO, B=Multiply(A=Multiply(B=@n19) {constA=1}, B=VertexNormalWS()))
  n20 = NamedRerouteUsage()
  n21 = Multiply(A=NamedRerouteUsage(), B=@n9)
  n22 = ScalarParameter() {defaultValue=50, parameterName=EndPower}
  n23 = ScalarParameter() {defaultValue=15, parameterName=StartPower}
  Multiply(
   A=NamedRerouteUsage()
   B=Power(Base=LinearInterpolate(B=@n20, Alpha=@n21) {constA=1}, Exp=LinearInterpolate(A=@n22, B=@n23, Alpha=@n20) {constA=50, constB=-1})
  )
  NamedRerouteDeclaration(None=@n4.Erode 1-0) {name=ErodeParam, nodeColor=<rgba 0 0.5312 1 1>}
  NamedRerouteDeclaration(None=@n18) {name=VerticalMask, nodeColor=<rgba 1 0 0.7734 1>}
  n24 = TextureObject() {samplerType=SAMPLERTYPE_Normal, texture=/Game/Effects/Fort_Effects/Textures/Tile/T_CellNoise_Tile_N}
  n25 = NamedRerouteUsage()
  n26 = Abs(None=VertexNormalWS())
  n27 = Constant3Vector() {constant=<rgba 0 0 1 0>}
  n28 = Multiply(A=Saturate(None=Multiply(A=@n25) {constB=2}), B=Multiply(A=DotProduct(A=@n26, B=@n27), B=@n18))
  Reroute(None=LinearInterpolate(
    A=Constant3Vector() {constant=<rgba 0 0 1 0>}
    B=MaterialFunctionCall(TextureObject=@n24, TextureSize=@n5, WorldPosition=@n6) {materialFunction=&WorldAlignedNormal}.Z Texture
    Alpha=Saturate(None=@n28)
   ) {constAlpha=1})
  ScalarParameter() {defaultValue=0.612}
  Transform(None=Constant3Vector() {constant=<rgba 1 1 0 0>}) {transformSourceType=TRANSFORMSOURCE_Local, transformType=TRANSFORM_Instance}
 ]
}
