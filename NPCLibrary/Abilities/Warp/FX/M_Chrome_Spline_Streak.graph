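# Masked, two-sided chrome streak material for Warp ability Niagara mesh particles, built on the LackLuster base with SplineThicken supplying width-scaled normals.
# Gradient UVs are stretched or tiled, narrowed by a panning sine wobble and distorted by a scrolling flowmap.
# The warped UVs drive a sphere normal map and a repeating SphereMask whose radius shrinks with erosion.
# Dynamic parameters expose erode, wobble, tile, separate, pan and width, while fresnel-shaped particle color feeds emissive.
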Material /NPCLibrary/Abilities/Warp/FX/M_Chrome_Spline_Streak
graph {
 properties {
  bTangentSpaceNormal=false
  bUsedWithNiagaraMeshParticles=true
  blendMode=BLEND_Masked
  pixelDepthOffsetMode=PDOM_Legacy
  twoSided=true
  usedShadingModels=Default Lit
 }
 nodes [
  n1 = MaterialFunctionCall(UV Channel=TextureCoordinate()) {materialFunction=&LinearGradient}
  n2 = NamedRerouteUsage()
  n3 = AppendVector(
   A=@n1.UGradient
   B=MaterialFunctionCall!(Center Point=Constant() {r=0.5}, 0-1 Gradient=@n1.VGradient, Scale Factor=@n2) {materialFunction=/Engine/Functions/Engine_MaterialFunctions02/Gradients/StretchGradient}.Resulting Gradient
  )
  n4 = AppendVector(A=@n1.UGradient, B=Multiply(A=@n1.VGradient, B=@n2))
  n5 = NamedRerouteUsage()
  n6 = OneMinus(None=@n1.VGradient)
  n7 = NamedRerouteUsage()
  n8 = Multiply(A=@n1.VGradient, B=Multiply(A=NamedRerouteUsage()) {constB=0.5})
  n9 = Multiply(A=Time(), B=ScalarParameter() {defaultValue=-0.2, parameterName=PanSpeed})
  n10 = DynamicParameter() {parameterIndex=1, paramNames [Pan,WidthMult,Param3,Param4]}
  n11 = Add()
  n12 = Fmod(A=@n11, B=Constant() {r=1})
  n13 = Add(A=Multiply(A=Add(A=@n9, B=@n10.Pan), B=LinearInterpolate(Alpha=@n12) {constA=1, constB=0.4}), B=@n11)
  n14 = ConstantBiasScale(None=Sine(None=Add(A=@n8, B=@n13)))
  n15 = ScalarParameter() {defaultValue=-0.2, parameterName=SineWobbleSizeSubtract}
  n16 = Add(
   A=MaterialFunctionCall(In=@n6, Power=LinearInterpolate(Alpha=@n7) {constA=0.5, constB=0}) {materialFunction=&SCurve}.Result
   B=Multiply(A=Multiply(A=@n14, B=@n15) {constB=0.1}, B=NamedRerouteUsage())
  )
  n17 = Constant() {r=1}
  n18 = NamedRerouteUsage()
  n19 = TextureCoordinate() {vTiling=0.5}
  n20 = AppendVector(A=Constant() {r=1}, B=NamedRerouteUsage())
  n21 = TextureSample(UVs=Panner(Coordinate=Multiply(A=@n19, B=@n20), Time=@n13) {speedX=0.1, speedY=-1}) {samplerType=SAMPLERTYPE_LinearColor, texture=/Game/Environments/Apollo/Sets/Jerky/Materials/T_JerkyFlowmap}
  n22 = Multiply(A=ConstantBiasScale(None=ComponentMask(None=@n21.RGB) {g=true, r=true}) {bias=-0.5, scale=2}) {constB=0.2}
  n23 = Add(
   A=MaterialFunctionCall(UVs=LinearInterpolate(A=@n3, B=@n4, Alpha=@n5), Texture Scale=AppendVector(A=@n16, B=@n17)) {materialFunction=&ScaleUVsByCenter}.UVs
   B=LinearInterpolate(A=Multiply(A=@n18, B=@n22), Alpha=@n6) {constB=0}
  )
  n24 = Multiply(A=@n10.WidthMult, B=ScalarParameter() {defaultValue=25})
  n25 = ScalarParameter() {defaultValue=5, parameterName=EndingNormalIntesnity}
  n26 = Constant() {r=1}
  n27 = DynamicParameter() {defaultValue=<rgba 0 0 7 0>, paramNames [Erode,Wobble,Tile,Separate]}
  n28 = Power(Base=NamedRerouteDeclaration(None=@n27.Erode) {name=ErodeParam, nodeColor=<rgba 0.7344 0 1 1>}) {constExponent=3}
  n29 = Multiply(
   A=TextureSample(UVs=@n23) {samplerType=SAMPLERTYPE_Normal, texture=&T_Sphere_N}.RGB
   B=LinearInterpolate(
    B=MaterialFunctionCall(R=@n25, G=@n25, B=@n26) {materialFunction=&AppendMany}.RGB
    Alpha=@n28
   ) {constA=1}
  )
  BreakMaterialAttributes!(Attr=MaterialFunctionCall!(
    bBlendMaterialAttributes
    MA_In
    UVsOverride=@n23
    Override Normals
    BlendMaterialAttributesAlpha
    BaseColorOverride
    WorldPositionOverride
    VertexNormalOverride=MaterialFunctionCall(WidthBase=@n24, WidthTip=@n24, AdditionalNormal=@n29) {materialFunction=&SplineThicken}.Normal
    SourceNormalsStrengthOverride
   ) {materialFunction=/Game/TechArt/Seasonal/LackLuster/MaterialFunctions/MF_LackLuster_Base}.MA Out)
  n30 = OneMinus(None=Fresnel(ExponentIn=ScalarParameter() {defaultValue=0.5, parameterName=EmissiveFresnelExp}) {baseReflectFraction=0, exponent=7})
  MaterialFunctionCall(Emissive=Multiply(A=ParticleColor().RGB, B=Power(Base=@n30) {constExponent=3})) {materialFunction=&MF_FX_EmissiveAdjustments}
  NamedRerouteDeclaration(None=@n27.Tile) {name=TileParam, nodeColor=<rgba 0.5469 0 1 1>}
  NamedRerouteDeclaration(None=@n27.Wobble) {name=WobbleParam, nodeColor=<rgba 0.7344 0 1 1>}
  NamedRerouteDeclaration(None=@n27.Separate) {name=SeparateParam, nodeColor=<rgba 0.9922 0 1 1>}
  n31 = MaterialFunctionCall(Float2=@n23) {materialFunction=&BreakOutFloat2Components}
  n32 = Constant() {r=2}
  SphereMask(A=AppendVector(A=@n31.R, B=Fmod(A=@n31.G, B=@n32)), B=Constant() {r=0.5}, Radius=LinearInterpolate(Alpha=@n28) {constA=0.5, constB=0})
 ]
}
