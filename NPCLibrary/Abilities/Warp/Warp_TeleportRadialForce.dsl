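; Warp teleport actor that carries a physics wash force field from its spawn point to a destination, as a client-side cosmetic effect.
; SetDestination writes NewDestination, whose OnRep runs only on non-authority machines and gates on High, Epic or Cinematic effects quality.
; It spawns and attaches FN_WashForce, then a timeline lerps the actor from its BeginPlay location to the destination and destroys both when finished.
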(event EventBeginPlay
  (Variables|Default|SetOriginalLocation (Transformation|GetActorLocation self)))

(event Collision|EventActorBeginOverlap
  (params (OtherActor :type Actor%20Object%20Reference)))

(event Custom|StartMovingForceActor
  (bind (_direction _alpha) (Timeline|MoveForceActor Play)
    (:Update
      (Transformation|SetActorLocation (Math|Vector|Lerp(Vector) (Variables|Default|GetOriginalLocation) (Variables|Default|GetNewDestination) _alpha) false true))
    (:Finished
      (Actor|DestroyActor (Variables|Default|GetPhysicalForceActor))
      (Actor|DestroyActor self))))

(event EventTick
  (params (DeltaSeconds :type Float%20%28single-precision%29)))

(event Custom|SetDestination
  (params (Destination :type Vector))
  (Variables|Default|SetNewDestination Destination))

(fn OnRep_NewDestination
  (Networking|SwitchHasAuthority
    (:Remote
      (Utilities|EffectsQualityCheck 0
        (:Cinematic
          _)
        (:Epic
          _)
        (:High
          _))
      (bind _returnvalue (Game|SpawnActorfromClass "/Game/Effects/Niagara/Blueprints/PhsyicsFields/FN_WashForce.FN_WashForce_C" (Transformation|GetActorTransform self) "Undefined" "MultiplyWithRoot" 0 7e+02 1.1 -999.0 0))
      (bind _output_get (Variables|Default|SetPhysicalForceActor _returnvalue))
      (Transformation|AttachActorToActor _output_get self "None" "SnapToTarget" "KeepRelative" "KeepRelative" true)
      (CallFunction|StartMovingForceActor))))

(fn ConstructionScript)
